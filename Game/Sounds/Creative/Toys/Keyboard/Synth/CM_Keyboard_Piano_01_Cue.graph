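# Sound cue for a Creative toy keyboard piano that plays one of 24 notes selected by the KeysIndex integer parameter.
# A Switch, shared by all quality levels, picks from four BigPiano C-based samples, played raw or through fixed-pitch Modulators at semitone ratios.
# The pitch sets suggest two octaves, with seven natural keys and five sharps each; the default slot is silent.
# It uses music-block attenuation, primes on load and caps concurrency by stopping the oldest voice.
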SoundCue /Game/Sounds/Creative/Toys/Keyboard/Synth/CM_Keyboard_Piano_01_Cue
root=QualityLevel_0
nodes [
 {id=Modulator_10, children [WavePlayer_84], properties {pitchMax=1.1225, pitchMin=1.1225, volumeMax=1, volumeMin=1}}
 {id=Modulator_11, children [WavePlayer_85], properties {pitchMax=1.2599, pitchMin=1.2599}}
 {id=Modulator_12, children [WavePlayer_86], properties {pitchMax=1.3348, pitchMin=1.3348}}
 {id=Modulator_13, children [WavePlayer_87], properties {pitchMax=1.4983, pitchMin=1.4983}}
 {id=Modulator_14, children [WavePlayer_88], properties {pitchMax=1.6818, pitchMin=1.6818}}
 {id=Modulator_15, children [WavePlayer_89], properties {pitchMax=1.8877, pitchMin=1.8877}}
 {id=Modulator_16, children [WavePlayer_90], properties {pitchMax=1.1225, pitchMin=1.1225, volumeMax=1, volumeMin=1}}
 {id=Modulator_17, children [WavePlayer_91], properties {pitchMax=1.3348, pitchMin=1.3348}}
 {id=Modulator_18, children [WavePlayer_92], properties {pitchMax=1.4983, pitchMin=1.4983}}
 {id=Modulator_19, children [WavePlayer_93], properties {pitchMax=1.6818, pitchMin=1.6818}}
 {id=Modulator_24, children [WavePlayer_94], properties {pitchMax=1.1225, pitchMin=1.1225, volumeMax=1, volumeMin=1}}
 {id=Modulator_25, children [WavePlayer_95], properties {pitchMax=1.3348, pitchMin=1.3348}}
 {id=Modulator_26, children [WavePlayer_96], properties {pitchMax=1.4983, pitchMin=1.4983}}
 {id=Modulator_27, children [WavePlayer_97], properties {pitchMax=1.6818, pitchMin=1.6818}}
 {id=Modulator_3, children [WavePlayer_77], properties {pitchMax=1.1225, pitchMin=1.1225, volumeMax=1, volumeMin=1}}
 {id=Modulator_4, children [WavePlayer_78], properties {pitchMax=1.2599, pitchMin=1.2599}}
 {id=Modulator_6, children [WavePlayer_79], properties {pitchMax=1.3348, pitchMin=1.3348}}
 {id=Modulator_7, children [WavePlayer_80], properties {pitchMax=1.4983, pitchMin=1.4983}}
 {id=Modulator_8, children [WavePlayer_81], properties {pitchMax=1.6818, pitchMin=1.6818}}
 {id=Modulator_9, children [WavePlayer_83], properties {pitchMax=1.8877, pitchMin=1.8877}}
 {id=QualityLevel_0, children [Switch_1,Switch_1,Switch_1,Switch_1]}
 {
  id=Switch_1
  children [
   null
   WavePlayer_61
   Modulator_3
   Modulator_4
   Modulator_6
   Modulator_7
   Modulator_8
   Modulator_9
   WavePlayer_62
   Modulator_10
   Modulator_11
   Modulator_12
   Modulator_13
   Modulator_14
   Modulator_15
   WavePlayer_59
   Modulator_16
   Modulator_17
   Modulator_18
   Modulator_19
   WavePlayer_60
   Modulator_24
   Modulator_25
   Modulator_26
   Modulator_27
  ]
  properties {intParameterName=KeysIndex}
 }
 {id=WavePlayer_59, properties {soundWaveAssetPtr=/Game/Athena/Sounds/Quests/BigPiano/New_Wavs/C_1}}
 {id=WavePlayer_60, properties {soundWaveAssetPtr=/Game/Athena/Sounds/Quests/BigPiano/New_Wavs/C_2}}
 {id=WavePlayer_61, properties {soundWaveAssetPtr=/Game/Athena/Sounds/Quests/BigPiano/New_Wavs/C1}}
 {id=WavePlayer_62, properties {soundWaveAssetPtr=/Game/Athena/Sounds/Quests/BigPiano/New_Wavs/C2}}
 {id=WavePlayer_77, properties {soundWaveAssetPtr=&C1}}
 {id=WavePlayer_78, properties {soundWaveAssetPtr=&C1}}
 {id=WavePlayer_79, properties {soundWaveAssetPtr=&C1}}
 {id=WavePlayer_80, properties {soundWaveAssetPtr=&C1}}
 {id=WavePlayer_81, properties {soundWaveAssetPtr=&C1}}
 {id=WavePlayer_83, properties {soundWaveAssetPtr=&C1}}
 {id=WavePlayer_84, properties {soundWaveAssetPtr=&C2}}
 {id=WavePlayer_85, properties {soundWaveAssetPtr=&C2}}
 {id=WavePlayer_86, properties {soundWaveAssetPtr=&C2}}
 {id=WavePlayer_87, properties {soundWaveAssetPtr=&C2}}
 {id=WavePlayer_88, properties {soundWaveAssetPtr=&C2}}
 {id=WavePlayer_89, properties {soundWaveAssetPtr=&C2}}
 {id=WavePlayer_90, properties {soundWaveAssetPtr=&C_1}}
 {id=WavePlayer_91, properties {soundWaveAssetPtr=&C_1}}
 {id=WavePlayer_92, properties {soundWaveAssetPtr=&C_1}}
 {id=WavePlayer_93, properties {soundWaveAssetPtr=&C_1}}
 {id=WavePlayer_94, properties {soundWaveAssetPtr=&C_2}}
 {id=WavePlayer_95, properties {soundWaveAssetPtr=&C_2}}
 {id=WavePlayer_96, properties {soundWaveAssetPtr=&C_2}}
 {id=WavePlayer_97, properties {soundWaveAssetPtr=&C_2}}
]
properties {
 attenuationSettings=/Game/Sounds/Creative/Attenuations/Fort_Creative_MusicBlocks_Att
 bPrimeOnLoad=true
 maxDistance=100000
 soundClassObject=&SFX
 volumeMultiplier=3
 concurrencyOverrides {maxCount=8, resolutionRule=StopOldest}
}
