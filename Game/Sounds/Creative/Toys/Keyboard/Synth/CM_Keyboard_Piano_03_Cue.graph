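# Sound cue for the Piano 03 synth voice of a Creative toy keyboard, playing one of 24 notes chosen by the KeysIndex parameter.
# A Switch maps each index to one of four base samples, with index 0 left empty and fixed-pitch Modulators shifting samples by semitone ratios to fill out the keys.
# All quality levels share that Switch.
# Playback uses music-block attenuation and capped concurrency that stops the oldest note.
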SoundCue /Game/Sounds/Creative/Toys/Keyboard/Synth/CM_Keyboard_Piano_03_Cue
root=QualityLevel_0
nodes [
 {id=Modulator_0, children [WavePlayer_1], properties {pitchMax=0.8409, pitchMin=0.8409}}
 {id=Modulator_1, children [WavePlayer_2], properties {pitchMax=0.9439, pitchMin=0.9439}}
 {id=Modulator_20, children [WavePlayer_5], properties {pitchMax=1.2599, pitchMin=1.2599}}
 {id=Modulator_21, children [WavePlayer_6], properties {pitchMax=1.4142, pitchMin=1.4142}}
 {id=Modulator_22, children [WavePlayer_8], properties {pitchMax=0.8409, pitchMin=0.8409}}
 {id=Modulator_23, children [WavePlayer_9], properties {pitchMax=0.9439, pitchMin=0.9439}}
 {id=Modulator_29, children [WavePlayer_11], properties {pitchMax=1.1225, pitchMin=1.1225}}
 {id=Modulator_3, children [WavePlayer_0], properties {pitchMax=0.7491, pitchMin=0.7491}}
 {id=Modulator_30, children [WavePlayer_12], properties {pitchMax=1.2599, pitchMin=1.2599}}
 {id=Modulator_31, children [WavePlayer_13], properties {pitchMax=1.4142, pitchMin=1.4142}}
 {id=Modulator_32, children [WavePlayer_15], properties {pitchMax=0.8409, pitchMin=0.8409}}
 {id=Modulator_34, children [WavePlayer_17], properties {pitchMax=1.1225, pitchMin=1.1225}}
 {id=Modulator_35, children [WavePlayer_18], properties {pitchMax=1.2599, pitchMin=1.2599}}
 {id=Modulator_36, children [WavePlayer_20], properties {pitchMax=0.8409, pitchMin=0.8409}}
 {id=Modulator_38, children [WavePlayer_22], properties {pitchMax=1.1225, pitchMin=1.1225}}
 {id=Modulator_39, children [WavePlayer_23], properties {pitchMax=1.2599, pitchMin=1.2599}}
 {id=Modulator_4, children [WavePlayer_7], properties {pitchMax=0.7491, pitchMin=0.7491}}
 {id=Modulator_5, children [WavePlayer_4], properties {pitchMax=1.1225, pitchMin=1.1225}}
 {id=Modulator_6, children [WavePlayer_14], properties {pitchMax=0.7491, pitchMin=0.7491}}
 {id=Modulator_7, children [WavePlayer_19], properties {pitchMax=0.7491, pitchMin=0.7491}}
 {id=QualityLevel_0, children [Switch_1,Switch_1,Switch_1,Switch_1]}
 {
  id=Switch_1
  children [
   null
   Modulator_3
   Modulator_0
   Modulator_1
   WavePlayer_3
   Modulator_5
   Modulator_20
   Modulator_21
   Modulator_4
   Modulator_22
   Modulator_23
   WavePlayer_10
   Modulator_29
   Modulator_30
   Modulator_31
   Modulator_6
   Modulator_32
   WavePlayer_16
   Modulator_34
   Modulator_35
   Modulator_7
   Modulator_36
   WavePlayer_21
   Modulator_38
   Modulator_39
  ]
  properties {intParameterName=KeysIndex}
 }
 {id=WavePlayer_0, properties {soundWaveAssetPtr=&/CM_Synth_Piano_03_F1}}
 {id=WavePlayer_1, properties {soundWaveAssetPtr=&CM_Synth_Piano_03_F1}}
 {id=WavePlayer_10, properties {soundWaveAssetPtr=&/CM_Synth_Piano_03_F2}}
 {id=WavePlayer_11, properties {soundWaveAssetPtr=&CM_Synth_Piano_03_F2}}
 {id=WavePlayer_12, properties {soundWaveAssetPtr=&CM_Synth_Piano_03_F2}}
 {id=WavePlayer_13, properties {soundWaveAssetPtr=&CM_Synth_Piano_03_F2}}
 {id=WavePlayer_14, properties {soundWaveAssetPtr=&/CM_Synth_Piano_03_F_1}}
 {id=WavePlayer_15, properties {soundWaveAssetPtr=&CM_Synth_Piano_03_F_1}}
 {id=WavePlayer_16, properties {soundWaveAssetPtr=&CM_Synth_Piano_03_F_1}}
 {id=WavePlayer_17, properties {soundWaveAssetPtr=&CM_Synth_Piano_03_F_1}}
 {id=WavePlayer_18, properties {soundWaveAssetPtr=&CM_Synth_Piano_03_F_1}}
 {id=WavePlayer_19, properties {soundWaveAssetPtr=&/CM_Synth_Piano_03_F_2}}
 {id=WavePlayer_2, properties {soundWaveAssetPtr=&CM_Synth_Piano_03_F1}}
 {id=WavePlayer_20, properties {soundWaveAssetPtr=&CM_Synth_Piano_03_F_2}}
 {id=WavePlayer_21, properties {soundWaveAssetPtr=&CM_Synth_Piano_03_F_2}}
 {id=WavePlayer_22, properties {soundWaveAssetPtr=&CM_Synth_Piano_03_F_2}}
 {id=WavePlayer_23, properties {soundWaveAssetPtr=&CM_Synth_Piano_03_F_2}}
 {id=WavePlayer_3, properties {soundWaveAssetPtr=&CM_Synth_Piano_03_F1}}
 {id=WavePlayer_4, properties {soundWaveAssetPtr=&CM_Synth_Piano_03_F1}}
 {id=WavePlayer_5, properties {soundWaveAssetPtr=&CM_Synth_Piano_03_F1}}
 {id=WavePlayer_6, properties {soundWaveAssetPtr=&CM_Synth_Piano_03_F1}}
 {id=WavePlayer_7, properties {soundWaveAssetPtr=&CM_Synth_Piano_03_F2}}
 {id=WavePlayer_8, properties {soundWaveAssetPtr=&CM_Synth_Piano_03_F2}}
 {id=WavePlayer_9, properties {soundWaveAssetPtr=&CM_Synth_Piano_03_F2}}
]
properties {
 attenuationSettings=/Game/Sounds/Creative/Attenuations/Fort_Creative_MusicBlocks_Papaya_Att
 bPrimeOnLoad=true
 maxDistance=4500
 soundClassObject=&SFX
 volumeMultiplier=1
 concurrencyOverrides {maxCount=8, resolutionRule=StopOldest}
}
