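# Sound cue for the synth lead voice of a Creative toy keyboard, playing one of 24 notes across two octaves.
# A Switch on the KeysIndex parameter, with an empty first slot, selects the note, and every quality level shares that switch.
# Four root samples are pitch-shifted by fixed semitone ratios: C1 and C2 form seven-note major scales, C_1 and C_2 five-note black-key sets.
# It uses music-block attenuation and stop-oldest concurrency.
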SoundCue /Game/Sounds/Creative/Toys/Keyboard/Synth/CM_Keyboard_Synth_Lead_01_Cue
root=QualityLevel_0
nodes [
 {id=Modulator_0, children [WavePlayer_1], properties {pitchMax=1.1225, pitchMin=1.1225, volumeMax=1, volumeMin=1}}
 {id=Modulator_1, children [WavePlayer_2], properties {pitchMax=1.2599, pitchMin=1.2599}}
 {id=Modulator_2, children [WavePlayer_3], properties {pitchMax=1.3348, pitchMin=1.3348}}
 {id=Modulator_20, children [WavePlayer_5], properties {pitchMax=1.6818, pitchMin=1.6818}}
 {id=Modulator_21, children [WavePlayer_6], properties {pitchMax=1.8877, pitchMin=1.8877}}
 {id=Modulator_22, children [WavePlayer_8], properties {pitchMax=1.1225, pitchMin=1.1225, volumeMax=1, volumeMin=1}}
 {id=Modulator_23, children [WavePlayer_9], properties {pitchMax=1.2599, pitchMin=1.2599}}
 {id=Modulator_28, children [WavePlayer_10], properties {pitchMax=1.3348, pitchMin=1.3348}}
 {id=Modulator_29, children [WavePlayer_11], properties {pitchMax=1.4983, pitchMin=1.4983}}
 {id=Modulator_30, children [WavePlayer_12], properties {pitchMax=1.6818, pitchMin=1.6818}}
 {id=Modulator_31, children [WavePlayer_13], properties {pitchMax=1.8877, pitchMin=1.8877}}
 {id=Modulator_32, children [WavePlayer_15], properties {pitchMax=1.1225, pitchMin=1.1225, volumeMax=1, volumeMin=1}}
 {id=Modulator_33, children [WavePlayer_16], properties {pitchMax=1.3348, pitchMin=1.3348}}
 {id=Modulator_34, children [WavePlayer_17], properties {pitchMax=1.4983, pitchMin=1.4983}}
 {id=Modulator_35, children [WavePlayer_18], properties {pitchMax=1.6818, pitchMin=1.6818}}
 {id=Modulator_36, children [WavePlayer_20], properties {pitchMax=1.1225, pitchMin=1.1225, volumeMax=1, volumeMin=1}}
 {id=Modulator_37, children [WavePlayer_21], properties {pitchMax=1.3348, pitchMin=1.3348}}
 {id=Modulator_38, children [WavePlayer_22], properties {pitchMax=1.4983, pitchMin=1.4983}}
 {id=Modulator_39, children [WavePlayer_23], properties {pitchMax=1.6818, pitchMin=1.6818}}
 {id=Modulator_5, children [WavePlayer_4], properties {pitchMax=1.4983, pitchMin=1.4983}}
 {id=QualityLevel_0, children [Switch_1,Switch_1,Switch_1,Switch_1]}
 {
  id=Switch_1
  children [
   null
   WavePlayer_0
   Modulator_0
   Modulator_1
   Modulator_2
   Modulator_5
   Modulator_20
   Modulator_21
   WavePlayer_7
   Modulator_22
   Modulator_23
   Modulator_28
   Modulator_29
   Modulator_30
   Modulator_31
   WavePlayer_14
   Modulator_32
   Modulator_33
   Modulator_34
   Modulator_35
   WavePlayer_19
   Modulator_36
   Modulator_37
   Modulator_38
   Modulator_39
  ]
  properties {intParameterName=KeysIndex}
 }
 {id=WavePlayer_0, properties {soundWaveAssetPtr=&/CM_Synth_Lead_01_C1}}
 {id=WavePlayer_1, properties {soundWaveAssetPtr=&CM_Synth_Lead_01_C1}}
 {id=WavePlayer_10, properties {soundWaveAssetPtr=&/CM_Synth_Lead_01_C2}}
 {id=WavePlayer_11, properties {soundWaveAssetPtr=&CM_Synth_Lead_01_C2}}
 {id=WavePlayer_12, properties {soundWaveAssetPtr=&CM_Synth_Lead_01_C2}}
 {id=WavePlayer_13, properties {soundWaveAssetPtr=&CM_Synth_Lead_01_C2}}
 {id=WavePlayer_14, properties {soundWaveAssetPtr=&/CM_Synth_Lead_01_C_1}}
 {id=WavePlayer_15, properties {soundWaveAssetPtr=&CM_Synth_Lead_01_C_1}}
 {id=WavePlayer_16, properties {soundWaveAssetPtr=&CM_Synth_Lead_01_C_1}}
 {id=WavePlayer_17, properties {soundWaveAssetPtr=&CM_Synth_Lead_01_C_1}}
 {id=WavePlayer_18, properties {soundWaveAssetPtr=&CM_Synth_Lead_01_C_1}}
 {id=WavePlayer_19, properties {soundWaveAssetPtr=&/CM_Synth_Lead_01_C_2}}
 {id=WavePlayer_2, properties {soundWaveAssetPtr=&CM_Synth_Lead_01_C1}}
 {id=WavePlayer_20, properties {soundWaveAssetPtr=&CM_Synth_Lead_01_C_2}}
 {id=WavePlayer_21, properties {soundWaveAssetPtr=&CM_Synth_Lead_01_C_2}}
 {id=WavePlayer_22, properties {soundWaveAssetPtr=&CM_Synth_Lead_01_C_2}}
 {id=WavePlayer_23, properties {soundWaveAssetPtr=&CM_Synth_Lead_01_C_2}}
 {id=WavePlayer_3, properties {soundWaveAssetPtr=&CM_Synth_Lead_01_C1}}
 {id=WavePlayer_4, properties {soundWaveAssetPtr=&CM_Synth_Lead_01_C1}}
 {id=WavePlayer_5, properties {soundWaveAssetPtr=&CM_Synth_Lead_01_C1}}
 {id=WavePlayer_6, properties {soundWaveAssetPtr=&CM_Synth_Lead_01_C1}}
 {id=WavePlayer_7, properties {soundWaveAssetPtr=&CM_Synth_Lead_01_C2}}
 {id=WavePlayer_8, properties {soundWaveAssetPtr=&CM_Synth_Lead_01_C2}}
 {id=WavePlayer_9, properties {soundWaveAssetPtr=&CM_Synth_Lead_01_C2}}
]
properties {
 attenuationSettings=/Game/Sounds/Creative/Attenuations/Fort_Creative_MusicBlocks_Att
 bPrimeOnLoad=true
 maxDistance=100000
 soundClassObject=&SFX
 volumeMultiplier=1.3
 concurrencyOverrides {maxCount=8, resolutionRule=StopOldest}
}
