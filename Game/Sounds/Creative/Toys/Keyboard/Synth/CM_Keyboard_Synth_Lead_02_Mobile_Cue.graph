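# Sound cue for the synth lead 02 voice of a Creative toy keyboard, mobile variant, playing one note per key.
# A KeysIndex switch leaves index 0 silent and maps 24 keys: two seven-note natural runs, then two five-note runs, apparently the sharps.
# Four root samples cover all keys through fixed-pitch modulators, and every quality level feeds the same switch.
# It uses music-block attenuation and caps concurrency at four voices, stopping the oldest.
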SoundCue /Game/Sounds/Creative/Toys/Keyboard/Synth/CM_Keyboard_Synth_Lead_02_Mobile_Cue
root=QualityLevel_1
nodes [
 {id=Modulator_0, children [WavePlayer_21], properties {pitchMax=1.1225, pitchMin=1.1225, volumeMax=1, volumeMin=1}}
 {id=Modulator_1, children [WavePlayer_22], properties {pitchMax=1.2599, pitchMin=1.2599}}
 {id=Modulator_2, children [WavePlayer_23], properties {pitchMax=1.3348, pitchMin=1.3348}}
 {id=Modulator_20, children [WavePlayer_50], properties {pitchMax=1.6818, pitchMin=1.6818}}
 {id=Modulator_21, children [WavePlayer_51], properties {pitchMax=1.8877, pitchMin=1.8877}}
 {id=Modulator_22, children [WavePlayer_53], properties {pitchMax=1.1225, pitchMin=1.1225, volumeMax=1, volumeMin=1}}
 {id=Modulator_23, children [WavePlayer_59], properties {pitchMax=1.2599, pitchMin=1.2599}}
 {id=Modulator_28, children [WavePlayer_60], properties {pitchMax=1.3348, pitchMin=1.3348}}
 {id=Modulator_29, children [WavePlayer_61], properties {pitchMax=1.4983, pitchMin=1.4983}}
 {id=Modulator_30, children [WavePlayer_62], properties {pitchMax=1.6818, pitchMin=1.6818}}
 {id=Modulator_31, children [WavePlayer_77], properties {pitchMax=1.8877, pitchMin=1.8877}}
 {id=Modulator_32, children [WavePlayer_79], properties {pitchMax=1.1225, pitchMin=1.1225, volumeMax=1, volumeMin=1}}
 {id=Modulator_33, children [WavePlayer_80], properties {pitchMax=1.3348, pitchMin=1.3348}}
 {id=Modulator_34, children [WavePlayer_81], properties {pitchMax=1.4983, pitchMin=1.4983}}
 {id=Modulator_35, children [WavePlayer_82], properties {pitchMax=1.6818, pitchMin=1.6818}}
 {id=Modulator_36, children [WavePlayer_84], properties {pitchMax=1.1225, pitchMin=1.1225, volumeMax=1, volumeMin=1}}
 {id=Modulator_37, children [WavePlayer_85], properties {pitchMax=1.3348, pitchMin=1.3348}}
 {id=Modulator_38, children [WavePlayer_86], properties {pitchMax=1.4983, pitchMin=1.4983}}
 {id=Modulator_39, children [WavePlayer_87], properties {pitchMax=1.6818, pitchMin=1.6818}}
 {id=Modulator_5, children [WavePlayer_49], properties {pitchMax=1.4983, pitchMin=1.4983}}
 {id=QualityLevel_1, children [Switch_0,Switch_0,Switch_0,Switch_0]}
 {
  id=Switch_0
  children [
   null
   WavePlayer_20
   Modulator_0
   Modulator_1
   Modulator_2
   Modulator_5
   Modulator_20
   Modulator_21
   WavePlayer_52
   Modulator_22
   Modulator_23
   Modulator_28
   Modulator_29
   Modulator_30
   Modulator_31
   WavePlayer_78
   Modulator_32
   Modulator_33
   Modulator_34
   Modulator_35
   WavePlayer_83
   Modulator_36
   Modulator_37
   Modulator_38
   Modulator_39
  ]
  properties {intParameterName=KeysIndex}
 }
 {id=WavePlayer_20, properties {soundWaveAssetPtr=&/CM_Synth_Lead_02_C1}}
 {id=WavePlayer_21, properties {soundWaveAssetPtr=&CM_Synth_Lead_02_C1}}
 {id=WavePlayer_22, properties {soundWaveAssetPtr=&CM_Synth_Lead_02_C1}}
 {id=WavePlayer_23, properties {soundWaveAssetPtr=&CM_Synth_Lead_02_C1}}
 {id=WavePlayer_49, properties {soundWaveAssetPtr=&CM_Synth_Lead_02_C1}}
 {id=WavePlayer_50, properties {soundWaveAssetPtr=&CM_Synth_Lead_02_C1}}
 {id=WavePlayer_51, properties {soundWaveAssetPtr=&CM_Synth_Lead_02_C1}}
 {id=WavePlayer_52, properties {soundWaveAssetPtr=&/CM_Synth_Lead_02_C2}}
 {id=WavePlayer_53, properties {soundWaveAssetPtr=&CM_Synth_Lead_02_C2}}
 {id=WavePlayer_59, properties {soundWaveAssetPtr=&CM_Synth_Lead_02_C2}}
 {id=WavePlayer_60, properties {soundWaveAssetPtr=&CM_Synth_Lead_02_C2}}
 {id=WavePlayer_61, properties {soundWaveAssetPtr=&CM_Synth_Lead_02_C2}}
 {id=WavePlayer_62, properties {soundWaveAssetPtr=&CM_Synth_Lead_02_C2}}
 {id=WavePlayer_77, properties {soundWaveAssetPtr=&CM_Synth_Lead_02_C2}}
 {id=WavePlayer_78, properties {soundWaveAssetPtr=&/CM_Synth_Lead_02_C_1}}
 {id=WavePlayer_79, properties {soundWaveAssetPtr=&CM_Synth_Lead_02_C_1}}
 {id=WavePlayer_80, properties {soundWaveAssetPtr=&CM_Synth_Lead_02_C_1}}
 {id=WavePlayer_81, properties {soundWaveAssetPtr=&CM_Synth_Lead_02_C_1}}
 {id=WavePlayer_82, properties {soundWaveAssetPtr=&CM_Synth_Lead_02_C_1}}
 {id=WavePlayer_83, properties {soundWaveAssetPtr=&/CM_Synth_Lead_02_C_2}}
 {id=WavePlayer_84, properties {soundWaveAssetPtr=&CM_Synth_Lead_02_C_2}}
 {id=WavePlayer_85, properties {soundWaveAssetPtr=&CM_Synth_Lead_02_C_2}}
 {id=WavePlayer_86, properties {soundWaveAssetPtr=&CM_Synth_Lead_02_C_2}}
 {id=WavePlayer_87, properties {soundWaveAssetPtr=&CM_Synth_Lead_02_C_2}}
]
properties {
 attenuationSettings=/Game/Sounds/Creative/Attenuations/Fort_Creative_MusicBlocks_Att
 bOverrideConcurrency=true
 bPrimeOnLoad=true
 maxDistance=100000
 soundClassObject=&SFX
 volumeMultiplier=1.3
 concurrencyOverrides {maxCount=4, resolutionRule=StopOldest}
}
