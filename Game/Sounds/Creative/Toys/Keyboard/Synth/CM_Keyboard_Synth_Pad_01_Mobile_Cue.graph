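# Sound cue for the mobile version of a Creative synth pad keyboard toy, playing one note per key.
# A switch on KeysIndex picks among 24 notes, with index 0 silent, and all four quality levels share that switch.
# Fixed-pitch modulators stretch four F-named samples into the notes, leaving each sample's root note unshifted.
# It primes on load, uses music-block attenuation and caps playback at four voices, stopping the oldest.
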SoundCue /Game/Sounds/Creative/Toys/Keyboard/Synth/CM_Keyboard_Synth_Pad_01_Mobile_Cue
root=QualityLevel_0
nodes [
 {id=Modulator_0, children [WavePlayer_25], properties {pitchMax=0.8409, pitchMin=0.8409}}
 {id=Modulator_1, children [WavePlayer_26], properties {pitchMax=0.9439, pitchMin=0.9439}}
 {id=Modulator_20, children [WavePlayer_29], properties {pitchMax=1.2599, pitchMin=1.2599}}
 {id=Modulator_21, children [WavePlayer_30], properties {pitchMax=1.4142, pitchMin=1.4142}}
 {id=Modulator_22, children [WavePlayer_32], properties {pitchMax=0.8409, pitchMin=0.8409}}
 {id=Modulator_23, children [WavePlayer_33], properties {pitchMax=0.9439, pitchMin=0.9439}}
 {id=Modulator_29, children [WavePlayer_35], properties {pitchMax=1.1225, pitchMin=1.1225}}
 {id=Modulator_3, children [WavePlayer_24], properties {pitchMax=0.7491, pitchMin=0.7491}}
 {id=Modulator_30, children [WavePlayer_36], properties {pitchMax=1.2599, pitchMin=1.2599}}
 {id=Modulator_31, children [WavePlayer_37], properties {pitchMax=1.4142, pitchMin=1.4142}}
 {id=Modulator_32, children [WavePlayer_39], properties {pitchMax=0.8409, pitchMin=0.8409}}
 {id=Modulator_34, children [WavePlayer_41], properties {pitchMax=1.1225, pitchMin=1.1225}}
 {id=Modulator_35, children [WavePlayer_42], properties {pitchMax=1.2599, pitchMin=1.2599}}
 {id=Modulator_36, children [WavePlayer_44], properties {pitchMax=0.8409, pitchMin=0.8409}}
 {id=Modulator_38, children [WavePlayer_46], properties {pitchMax=1.1225, pitchMin=1.1225}}
 {id=Modulator_39, children [WavePlayer_47], properties {pitchMax=1.2599, pitchMin=1.2599}}
 {id=Modulator_4, children [WavePlayer_31], properties {pitchMax=0.7491, pitchMin=0.7491}}
 {id=Modulator_5, children [WavePlayer_28], properties {pitchMax=1.1225, pitchMin=1.1225}}
 {id=Modulator_6, children [WavePlayer_38], properties {pitchMax=0.7491, pitchMin=0.7491}}
 {id=Modulator_7, children [WavePlayer_43], properties {pitchMax=0.7491, pitchMin=0.7491}}
 {id=QualityLevel_0, children [Switch_1,Switch_1,Switch_1,Switch_1]}
 {
  id=Switch_1
  children [
   null
   Modulator_3
   Modulator_0
   Modulator_1
   WavePlayer_27
   Modulator_5
   Modulator_20
   Modulator_21
   Modulator_4
   Modulator_22
   Modulator_23
   WavePlayer_34
   Modulator_29
   Modulator_30
   Modulator_31
   Modulator_6
   Modulator_32
   WavePlayer_40
   Modulator_34
   Modulator_35
   Modulator_7
   Modulator_36
   WavePlayer_45
   Modulator_38
   Modulator_39
  ]
  properties {intParameterName=KeysIndex}
 }
 {id=WavePlayer_24, properties {soundWaveAssetPtr=&/CM_Synth_Pad_01_F1}}
 {id=WavePlayer_25, properties {soundWaveAssetPtr=&CM_Synth_Pad_01_F1}}
 {id=WavePlayer_26, properties {soundWaveAssetPtr=&CM_Synth_Pad_01_F1}}
 {id=WavePlayer_27, properties {soundWaveAssetPtr=&CM_Synth_Pad_01_F1}}
 {id=WavePlayer_28, properties {soundWaveAssetPtr=&CM_Synth_Pad_01_F1}}
 {id=WavePlayer_29, properties {soundWaveAssetPtr=&CM_Synth_Pad_01_F1}}
 {id=WavePlayer_30, properties {soundWaveAssetPtr=&CM_Synth_Pad_01_F1}}
 {id=WavePlayer_31, properties {soundWaveAssetPtr=&/CM_Synth_Pad_01_F2}}
 {id=WavePlayer_32, properties {soundWaveAssetPtr=&CM_Synth_Pad_01_F2}}
 {id=WavePlayer_33, properties {soundWaveAssetPtr=&CM_Synth_Pad_01_F2}}
 {id=WavePlayer_34, properties {soundWaveAssetPtr=&CM_Synth_Pad_01_F2}}
 {id=WavePlayer_35, properties {soundWaveAssetPtr=&CM_Synth_Pad_01_F2}}
 {id=WavePlayer_36, properties {soundWaveAssetPtr=&CM_Synth_Pad_01_F2}}
 {id=WavePlayer_37, properties {soundWaveAssetPtr=&CM_Synth_Pad_01_F2}}
 {id=WavePlayer_38, properties {soundWaveAssetPtr=&/CM_Synth_Pad_01_F_1}}
 {id=WavePlayer_39, properties {soundWaveAssetPtr=&CM_Synth_Pad_01_F_1}}
 {id=WavePlayer_40, properties {soundWaveAssetPtr=&CM_Synth_Pad_01_F_1}}
 {id=WavePlayer_41, properties {soundWaveAssetPtr=&CM_Synth_Pad_01_F_1}}
 {id=WavePlayer_42, properties {soundWaveAssetPtr=&CM_Synth_Pad_01_F_1}}
 {id=WavePlayer_43, properties {soundWaveAssetPtr=&/CM_Synth_Pad_01_F_2}}
 {id=WavePlayer_44, properties {soundWaveAssetPtr=&CM_Synth_Pad_01_F_2}}
 {id=WavePlayer_45, properties {soundWaveAssetPtr=&CM_Synth_Pad_01_F_2}}
 {id=WavePlayer_46, properties {soundWaveAssetPtr=&CM_Synth_Pad_01_F_2}}
 {id=WavePlayer_47, properties {soundWaveAssetPtr=&CM_Synth_Pad_01_F_2}}
]
properties {
 attenuationSettings=/Game/Sounds/Creative/Attenuations/Fort_Creative_MusicBlocks_Att
 bOverrideConcurrency=true
 bPrimeOnLoad=true
 maxDistance=100000
 soundClassObject=&SFX
 volumeMultiplier=1.3
 concurrencyOverrides {maxCount=4, resolutionRule=StopOldest}
}
